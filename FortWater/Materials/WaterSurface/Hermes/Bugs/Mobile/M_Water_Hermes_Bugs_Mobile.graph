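# Unlit translucent mobile water surface for the Hermes Bugs variant, with animated patterns, edge foam and time-of-day tinting.
# Scene depth and WaterBodyData depth drive the foam mask, with an IsStaticMesh switch bypassing the water-body blend and height offset.
# A scrolling texture through a sine creates wobble patterns and derived normals.
# These feed a Fresnel-masked cubemap reflection, sun specular and emissive colors compensated by EyeAdaptationInverse.
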Material /FortWater/Materials/WaterSurface/Hermes/Bugs/Mobile/M_Water_Hermes_Bugs_Mobile
graph {
 properties {
  allowTranslucentCustomDepthWrites=true
  bTangentSpaceNormal=false
  bUsedWithInstancedStaticMeshes=true
  bUsedWithNanite=true
  bUsedWithSplineMeshes=true
  bUsedWithWater=true
  blendMode=BLEND_Translucent
  pixelDepthOffsetMode=PDOM_Legacy
  refractionMethod=RM_PixelNormalOffset
  shadingModel=MSM_Unlit
  usedShadingModels=Unlit
 }
 nodes [
  n1 = NamedRerouteUsage()
  Multiply(A=Add(A=LinearInterpolate(Alpha=@n1) {constA=0.55, constB=0.2}, B=NamedRerouteUsage()), B=NamedRerouteUsage()) {constB=0.8}
  n2 = VectorParameter() {parameterName=Mobile Water, defaultValue=<rgba 0.0267 0.7031 0.8542 1>}
  n3 = Multiply(A=@n2.RGB) {constB=0.5}
  n4 = NamedRerouteUsage()
  n5 = Multiply(A=NamedRerouteUsage(), B=ScalarParameter() {defaultValue=0.06, parameterName=Scattering Offset}) {constB=0.1}
  n6 = LinearInterpolate(A=Multiply(A=LinearInterpolate(A=@n2.RGB, B=@n3, Alpha=@n4), B=Add(A=@n5)), B=NamedRerouteUsage(), Alpha=NamedRerouteUsage())
  n7 = VectorParameter() {parameterName=Mobile Foam, defaultValue=<rgba 0.2483 0.8428 0.9167 1>}
  n8 = NamedRerouteUsage()
  Multiply(A=Add(A=LinearInterpolate(A=@n6, B=@n7.RGB, Alpha=@n8), B=NamedRerouteUsage()), B=NamedRerouteUsage())
  n9 = MaterialFunctionCall!(Texture, Position) {materialFunction=/FortWater/Materials/Functions/WaterBodyData}
  NamedRerouteDeclaration(None=@n9.Water Depth) {name=Water Depth, nodeColor=<rgba 0 0.0152 0.1095 1>}
  NamedRerouteDeclaration(None=@n9.Water Z Height) {name=Water Z Height, nodeColor=<rgba 0 1 0.2266 1>}
  NamedRerouteDeclaration(None=@n9.Terrain Z Height) {name=Terrain Z Height, nodeColor=<rgba 1 0 0.0938 1>}
  n10 = NamedRerouteUsage()
  n11 = WorldPosition()
  NamedRerouteDeclaration(None=StaticSwitchParameter(True=Constant(), False=Subtract(A=@n10, B=@n11.Z)) {parameterName=IsStaticMesh}) {name=WPO, nodeColor=<rgba 0 0.4141 1 1>}
  n12 = CameraVectorWS()
  n13 = Multiply(A=Subtract(A=SceneDepth().R, B=PixelDepth().R), B=Abs(None=ComponentMask(None=@n12) {b=true}))
  n14 = NamedRerouteUsage()
  n15 = Multiply(A=Add(A=@n14) {constB=10}) {constB=1000}
  n16 = PixelDepth()
  n17 = Saturate(None=Divide(A=Subtract(A=@n16.R) {constB=5000}) {constB=10000})
  n18 = StaticSwitchParameter(
   True=@n13
   False=LinearInterpolate(A=Min(A=@n13, B=@n15), B=@n14, Alpha=NamedRerouteDeclaration(None=@n17) {name=Pixel Depth, nodeColor=<rgba 0.2891 1 0 1>})
  ) {parameterName=IsStaticMesh}
  n19 = Saturate(None=Divide(A=@n18, B=ScalarParameter() {defaultValue=40, parameterName=Foam Range}) {constB=44})
  n20 = OneMinus(None=Saturate(None=Divide(A=@n18) {constB=10}))
  n21 = OneMinus(None=Multiply(A=@n20, B=@n20))
  n22 = Multiply(
   A=NamedRerouteDeclaration(None=OneMinus(None=@n19)) {name=Edge Mask, nodeColor=<rgba 1 0.6699 0 1>}
   B=NamedRerouteDeclaration(None=Multiply(A=@n21, B=@n21)) {name=Edge Fade, nodeColor=<rgba 0.0781 0 1 1>}
  )
  n23 = Divide(A=WorldPosition().XY, B=ScalarParameter() {defaultValue=500, parameterName=Foam Scale}) {constB=800}
  n24 = NamedRerouteUsage()
  n25 = Multiply(A=Add(A=ComponentMask(None=@n24) {g=true, r=true}) {constB=-0.5}) {constB=0.1}
  n26 = TextureSample(UVs=NamedRerouteDeclaration(None=Add(A=@n23, B=@n25)) {name=UVs, nodeColor=<rgba 1 0 0.1641 1>}) {samplerType=SAMPLERTYPE_LinearColor, texture=/FortWater/Materials/WaterSurface/Hermes/Bugs/T_Water_Hermes_Foam_M}
  n27 = NamedRerouteUsage()
  n28 = ScalarParameter() {defaultValue=0.2, parameterName=Mask Offset}
  n29 = Multiply(A=Add(A=OneMinus(None=@n26.R), B=Multiply(A=@n27, B=@n28) {constB=0.15}), B=NamedRerouteUsage())
  NamedRerouteDeclaration(None=Saturate(None=Subtract(A=@n22, B=@n29))) {name=Foam Mask, nodeColor=<rgba 1 0 0.4219 1>}
  n30 = VectorParameter() {parameterName=Emissive Morning, defaultValue=<rgba 0 0.0365 0.009 1>}
  n31 = NamedRerouteUsage()
  n32 = ScalarParameter() {defaultValue=0.15, parameterName=Emission Offset}
  n33 = Multiply(
   A=MaterialFunctionCall(
    Color_Morning=@n30.RGB
    Color_Day=VectorParameter() {parameterName=Emissive Day, defaultValue=<rgba 0 0.0081 0.0104 1>}.RGB
    Color_Evening=@n30.RGB
    Color_Night=VectorParameter() {parameterName=Emissive Night, defaultValue=<rgba 0 0.097 0.1354 1>}.RGB
   ) {materialFunction=&Color_Time_of_Day}.Result
   B=Add(A=Multiply(A=@n31, B=@n32))
  )
  n34 = NamedRerouteUsage()
  NamedRerouteDeclaration(None=EyeAdaptationInverse(LightValueInput=Multiply(A=@n33, B=@n34), AlphaInput=ScalarParameter() {defaultValue=0.5, parameterName=EyeAdaptationInverse}).EyeAdaptationInverse) {name=Emission, nodeColor=<rgba 1 0.1172 0 1>}
  n35 = Add(
   A=Divide(A=WorldPosition().XY, B=ScalarParameter() {defaultValue=5000, parameterName=Offset Scale}) {constB=800}
   B=Multiply(A=Time(), B=Constant2Vector() {g=0.012, r=0.01}) {constB=0.01}
  )
  n36 = Multiply(
   A=NamedRerouteDeclaration(None=TextureSample(UVs=@n35) {samplerType=SAMPLERTYPE_LinearColor, texture=/FortWater/Materials/WaterSurface/Hermes/Bugs/T_Water_Hermes_Bugs_Offset}.RGBA) {name=Offset Texture, nodeColor=<rgba 1 0 0.8438 1>}
   B=ScalarParameter() {defaultValue=1.7, parameterName=Offset Pattern intensity}
  ) {constB=2}
  n37 = Multiply(A=NamedRerouteUsage(), B=ScalarParameter() {defaultValue=2, parameterName=Offset Edge intensity})
  n38 = Time()
  n39 = VectorParameter() {parameterName=Offset Speed, defaultValue=<rgba 0.1 0.25 0.45 0.55>}
  n40 = AppendVector(A=@n39.RGB, B=@n39.RGB)
  MaterialExpressionSine_1 = Sine(None=Add(A=Subtract(A=@n36, B=@n37), B=Multiply(A=@n38, B=@n40) {constB=0.2}))
  n41 = Convert(None=@MaterialExpressionSine_1) {
   nodeName=Break Float3
   convertInputs [{
     type=Vector3
     expressionInput {
      expression=&self:MaterialExpressionSine_1
      inputName=None
      mask=0
      maskA=0
      maskB=0
      maskG=0
      maskR=0
      outputIndex=0
     }
    }]
   convertOutputs [{type=Scalar}, {type=Scalar}]
  }
  NamedRerouteDeclaration(None=Add(A=@n41, B=@n41)) {name=Offset Slow Pattern, nodeColor=<rgba 1 0.4219 0 1>}
  NamedRerouteDeclaration(None=Add(A=ComponentMask(None=@MaterialExpressionSine_1) {a=true, b=true}, B=ComponentMask(None=@MaterialExpressionSine_1) {r=true})) {name=Offset Normal Pattern, nodeColor=<rgba 0 1 0.4609 1>}
  n42 = NamedRerouteUsage()
  n43 = CameraVectorWS()
  n44 = ReflectionVectorWS(CustomWorldNormal=Subtract(A=@n42, B=Multiply(A=@n43) {constB=0.14}) {constB=0}) {bNormalizeCustomWorldNormal=true}
  NamedRerouteDeclaration(None=Multiply(
    A=TextureSample(UVs=@n44) {samplerType=SAMPLERTYPE_LinearColor, texture=/FortWater/Materials/WaterSurface/Mobile/Textures/T_MobileWater_Cubemap}.RGB
    B=VectorParameter() {parameterName=Reflection Color, defaultValue=<rgba 0.6146 0.6146 0.6146 1>}.RGB
   ) {constB=0.75}) {name=Reflection Color, nodeColor=<rgba 0 0.2297 1 1>}
  n45 = CameraVectorWS()
  n46 = OneMinus(None=Saturate(None=DotProduct(A=@n42, B=@n45)))
  n47 = ScalarParameter() {defaultValue=2, group=Mobile Water, parameterName=Mobile Reflection MaskPower}
  NamedRerouteDeclaration(None=LinearInterpolate(Alpha=Power(Base=@n46, Exp=@n47) {constExponent=4}) {constA=0.3, constB=0.73}) {name=Reflection Mask, nodeColor=<rgba 1 0.0984 0.7042 1>}
  n48 = NamedRerouteUsage()
  n49 = ScalarParameter() {defaultValue=0.01, parameterName=Normal Intensity}
  NamedRerouteDeclaration(None=DeriveNormalZ(InXY=Multiply(A=@n48, B=@n49) {constB=0.05})) {name=Normals, nodeColor=<rgba 0.5703 0 1 1>}
  n50 = Saturate(None=Divide(A=PixelDepth().R) {constB=5000})
  n51 = ComponentMask(None=CameraVectorWS()) {b=true}
  NamedRerouteDeclaration(None=Multiply(A=OneMinus(None=@n50), B=Abs(None=@n51))) {name=Bubbles Fade, nodeColor=<rgba 0.2891 1 0 1>}
  n52 = SkyAtmosphereLightDirection()
  n53 = AppendVector(A=ComponentMask(None=@n52) {g=true, r=true}, B=LinearInterpolate(Alpha=ComponentMask(None=@n52) {b=true}) {constA=0.5})
  n54 = CameraVectorWS()
  n55 = DotProduct(A=Normalize(VectorInput=Add(A=@n53, B=@n54)), B=NamedRerouteUsage())
  n56 = Multiply(A=Power(Base=@n55) {constExponent=128}, B=ScalarParameter() {defaultValue=0.4, group=Mobile Water, parameterName=Mobile Specular Diffusion}) {constB=0.02}
  n57 = OneMinus(None=@n55)
  n58 = OneMinus(None=Saturate(None=Multiply(A=@n57) {constB=500}))
  n59 = Multiply(A=@n58, B=@n58)
  n60 = Multiply(A=@n59, B=@n59)
  n61 = Multiply(A=@n60, B=@n60)
  NamedRerouteDeclaration(None=Multiply(A=Add(A=@n56, B=@n61), B=NamedRerouteUsage())) {name=Sun Specular, nodeColor=<rgba 1 0.1699 0 1>}
  n62 = VectorParameter() {parameterName=Color Mult Morning, defaultValue=<rgba 1 0.8426 0.7344 1>}
  NamedRerouteDeclaration(None=MaterialFunctionCall(
    Color_Morning=@n62.RGB
    Color_Day=VectorParameter() {parameterName=Color Mult Day, defaultValue=<rgba 1 1 1 1>}.RGB
    Color_Evening=@n62.RGB
    Color_Night=VectorParameter() {parameterName=Color Mult Night, defaultValue=<rgba 0.557 0.6038 0.9062 1>}.RGB
   ) {materialFunction=&Color_Time_of_Day}.Result) {name=Color Mult, nodeColor=<rgba 0 0.7422 1 1>}
  NamedRerouteDeclaration(None=EyeAdaptationInverse(
    LightValueInput=VectorParameter() {parameterName=Foam Color, defaultValue=<rgba 0.0222 0.4005 0.4271 1>}.RGB
    AlphaInput=ScalarParameter() {defaultValue=0.5, parameterName=EyeAdaptationInverse}
   ).EyeAdaptationInverse) {name=Foam Color, nodeColor=<rgba 1 0.2344 0 1>}
  n63 = VectorParameter() {parameterName=Fallback Morning, defaultValue=<rgba 0.0024 0.2143 0.2292 1>}
  n64 = VectorParameter() {parameterName=Fallback Day, defaultValue=<rgba 0.0018 0.1607 0.1719 1>}
  n65 = VectorParameter() {parameterName=Fallback Night, defaultValue=<rgba 0.0035 0.3165 0.3385 1>}
  n66 = Add(
   A=LinearInterpolate(
    A=MaterialFunctionCall(Color_Morning=@n63.RGB, Color_Day=@n64.RGB, Color_Evening=@n63.RGB, Color_Night=@n65.RGB) {materialFunction=&Color_Time_of_Day}.Result
    B=VectorParameter() {parameterName=Fallback Edge, defaultValue=<rgba 0.0241 0.1406 0.1023 1>}.RGB
    Alpha=NamedRerouteUsage()
   )
   B=Multiply(A=NamedRerouteUsage(), B=ScalarParameter() {defaultValue=0.01, parameterName=Fallback Offset}) {constB=0.1}
  )
  n67 = NamedRerouteUsage()
  NamedRerouteDeclaration(None=EyeAdaptationInverse(LightValueInput=Subtract(A=@n66, B=@n67), AlphaInput=ScalarParameter() {defaultValue=0.5, parameterName=EyeAdaptationInverse}).EyeAdaptationInverse) {name=FallBack Color, nodeColor=<rgba 1 0.2344 0 1>}
  NamedRerouteUsage()
 ]
}
